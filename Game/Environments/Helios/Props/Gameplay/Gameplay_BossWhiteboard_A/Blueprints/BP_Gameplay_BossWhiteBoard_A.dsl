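; Gameplay boss whiteboard prop that reveals a plane for each eliminated boss: Bengal Basher, Ceremonial Guard, Diamond Heart and Madame Moth.
; Each boss has an OnRep function for its eliminated state, which unhides that boss's plane component without checking the replicated value.
; BeginPlay, overlap and Tick events are empty, and the construction script only calls its parent.
; Nothing in this graph sets the eliminated states or hides the planes again.
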(event EventBeginPlay)

(event Collision|EventActorBeginOverlap
  (params (OtherActor :type Actor%20Object%20Reference)))

(event EventTick
  (params (DeltaSeconds :type Float%20%28single-precision%29)))

(fn OnRep_BengalBasherEliminated
  (Development|SetHiddeninGame (Variables|Default|GetPlane) false false))

(fn OnRep_CeremonialGuardEliminated
  (Development|SetHiddeninGame (Variables|Default|GetPlane3) false false))

(fn OnRep_DiamondHeartEliminated
  (Development|SetHiddeninGame (Variables|Default|GetPlane4) false false))

(fn OnRep_MadameMothEliminated
  (Development|SetHiddeninGame (Variables|Default|GetPlane1) false false))

(fn ConstructionScript
  (|CallParentFunction))
