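# M_CrossMark is an unlit, masked material for cross marks on the boss whiteboard gameplay prop.
# It samples the T_Cross texture, which packs cross patterns into separate color channels.
# The CrossPattern scalar parameter drives a Switch that selects among those channels, letting one texture supply several mark variants.
# A constant dark red color is defined alongside, though connections to the material outputs are not shown.
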Material /Game/Environments/Helios/Props/Gameplay/Gameplay_BossWhiteboard_A/Materials/M_CrossMark
graph {
 expressions {
  MaterialExpressionConstant3Vector_1 {class=MaterialExpressionConstant3Vector, properties {constant=<rgba 0.5156 0 0.0041 1>}}
  MaterialExpressionScalarParameter_0 {class=MaterialExpressionScalarParameter, properties {parameterName=CrossPattern}}
  MaterialExpressionSwitch_0 {
   class=MaterialExpressionSwitch
   inputs [
    SwitchValue {from=MaterialExpressionScalarParameter_0, output=""}
    Default {from="", output=""}
    None {from=MaterialExpressionTextureSample_0, output=R}
    None {from=MaterialExpressionTextureSample_0, output=R}
    None {from=MaterialExpressionTextureSample_0, output=R}
   ]
   properties {inputs [
     {
      inputName=None
      input {
       expression=&self:MaterialExpressionTextureSample_0
       inputName=None
       mask=1
       maskA=0
       maskB=0
       maskG=0
       maskR=1
       outputIndex=1
      }
     }
     {
      inputName=None
      input {
       expression=&self:MaterialExpressionTextureSample_0
       inputName=None
       mask=1
       maskA=0
       maskB=0
       maskG=1
       maskR=0
       outputIndex=2
      }
     }
     {
      inputName=None
      input {
       expression=&self:MaterialExpressionTextureSample_0
       inputName=None
       mask=1
       maskA=0
       maskB=1
       maskG=0
       maskR=0
       outputIndex=3
      }
     }
    ]}
  }
  MaterialExpressionTextureSample_0 {
   class=MaterialExpressionTextureSample
   inputs [
    UVs {from=""}
    Tex {from=""}
    Apply View MipBias {from=""}
   ]
   properties {texture=/Game/Environments/Helios/Props/Gameplay/Gameplay_BossWhiteboard_A/Textures/T_Cross}
  }
 }
 properties {blendMode=BLEND_Masked, pixelDepthOffsetMode=PDOM_Legacy, shadingModel=MSM_Unlit, usedShadingModels=Unlit}
}
